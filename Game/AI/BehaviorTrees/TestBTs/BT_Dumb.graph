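# BT_Dumb is a test behavior tree on the FortAIBB_Default blackboard that keeps an AI idle, with no movement or combat tasks.
# The root selector first tries a sub-selector whose service sets default focus to GoalActor and whose five-second Wait is gated by a GoalActor blackboard condition.
# When that branch fails, the root falls back to an hour-long Wait.
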BehaviorTree /Game/AI/BehaviorTrees/TestBTs/BT_Dumb
blackboard=/Game/AI/Blackboards/FortAIBB_Default
root {
 name=Root
 class=BTComposite_Selector
 children [
  Selector {
   class=BTComposite_Selector
   children [Wait {
     class=BTTask_Wait
     decorators [Blackboard Based Condition {
       class=BTDecorator_Blackboard
       properties {arithmeticOperation=Equal, intValue=0, nodeName=Blackboard Based Condition, stringValue="", blackboardKey {selectedKeyName=GoalActor}}
      }]
     properties {waitTime {defaultValue=5}}
    }]
   decorators []
   properties {
    children [{
      childComposite=null
      childTask=&self:BTTask_Wait_12
      decorators [&self:BTDecorator_Blackboard_45]
     }]
    services [&self:BTService_DefaultFocus_0]
   }
   services [Set default focus {
     class=BTService_DefaultFocus
     properties {
      bCallTickOnSearchStart=false
      bRestartTimerOnEachActivation=false
      interval=0.5
      nodeName=Set default focus
      randomDeviation=0.1
      blackboardKey {bNoneIsAllowedValue=false, selectedKeyName=GoalActor}
     }
    }]
  }
  Wait an hour {class=BTTask_Wait, properties {nodeName=Wait an hour, waitTime {defaultValue=3600}}}
 ]
 decorators []
 properties {
  nodeName=Root
  children [
   {childComposite=&self:BTComposite_Selector_42, childTask=null, decorators []}
   {childComposite=null, childTask=&self:BTTask_Wait_2, decorators []}
  ]
  services []
 }
 services []
}
